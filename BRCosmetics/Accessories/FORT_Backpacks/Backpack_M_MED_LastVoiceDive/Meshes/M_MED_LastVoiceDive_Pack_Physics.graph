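# Physics asset for the LastVoiceDive backpack cosmetic mesh, giving its three dyn_box bones simple collision bodies for physics.
# Each bone carries one equally sized sphere, and two constraints link them in a chain from dyn_box_1 through dyn_box_2 to dyn_box_3.
# Both constraints lock swing and twist, so the bodies cannot rotate relative to one another despite the stored angle limits.
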PhysicsAsset /BRCosmetics/Accessories/FORT_Backpacks/Backpack_M_MED_LastVoiceDive/Meshes/M_MED_LastVoiceDive_Pack_Physics
bodies {
 dyn_box_1 [{shapeType=Sphere, radius=5}]
 dyn_box_2 [{shapeType=Sphere, radius=5, shapeName=dyn_box_2_sphere}]
 dyn_box_3 [{shapeType=Sphere, radius=5}]
}
constraints [
 {
  bone1Name=dyn_box_3
  bone2Name=dyn_box_2
  swing1LimitDegrees=45
  swing1Motion=Locked
  swing2LimitDegrees=45
  swing2Motion=Locked
  twistLimitDegrees=45
  twistMotion=Locked
 }
 {
  bone1Name=dyn_box_2
  bone2Name=dyn_box_1
  swing1LimitDegrees=45
  swing1Motion=Locked
  swing2LimitDegrees=45
  swing2Motion=Locked
  twistLimitDegrees=45
  twistMotion=Locked
 }
]
